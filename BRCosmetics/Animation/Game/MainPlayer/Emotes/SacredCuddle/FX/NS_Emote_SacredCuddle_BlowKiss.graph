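# Niagara effect for the SacredCuddle emote's blown kiss, launching a pink heart with small trailing dots.
# BigHeart bursts a mesh and sprite heart that grows, drags and drifts upward, while FN_Contrails_Dots bursts a few short-lived sprites stirred by curl noise.
# Both follow source movement with influence fading over the second half of life, fade alpha at the end and tint by time of day.
# The Empty emitter is disabled.
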NiagaraSystem /BRCosmetics/Animation/Game/MainPlayer/Emotes/SacredCuddle/FX/NS_Emote_SacredCuddle_BlowKiss
dependencies {
 usedDataInterfaces [&NiagaraDataInterfaceCurve, &NiagaraDataInterfaceVectorCurve, &NiagaraDataInterfaceArrayFloat]
 usedDynamicInputs [&SimulationPosition, &UniformRangedVector2D]
 usedModules [&SystemState]
}
emitters {
 BigHeart {
  emitterUpdateScript [
   &EmitterState(Enable Distance Culling=true, MaxDistance=4500, Module.bUseMaxDistance=true, Module.bUseMinDistance=false, Scalability Mode=::NewEnumerator1)
   &SpawnBurst_Instantaneous(Use Spawn Probability=false, UseLoopCountLimit=false)
  ]
  particleSpawnScript [
   &InitializeParticle(
    Mesh Scale=<0.35 0.35 0.35>
    Mesh Scale Mode=::NewEnumerator4
    Sprite Rotation Angle=&RandomRangeFloat(Minimum=-30, Maximum=30)
    Sprite Rotation Mode=::NewEnumerator0
    Sprite Size Mode=::NewEnumerator1
    Uniform Sprite Size Max=20
    UsePositionOffset=false
    Write Material Random=true
   )
   &AddVelocity(Velocity=<0 150 0>)
  ]
  particleUpdateScript [
   &ParticleState()
   &ScaleSpriteSize(Uniform Curve Sprite Scale=DataInterface() {propertyValues {Curve {keys [{tangentMode=RCTM_User, time=0, value=0.4}, {tangentMode=RCTM_User, time=1, value=1}]}}})
   &Drag(Drag=2, Use Linear Drag=true, Use Rotational Drag=true)
   &AccelerationForce(Acceleration=<0 0 80>)
   &InheritSourceMovement(
    Applied Position Scale=&VectorFromCurve(VectorCurve=DataInterface() {propertyValues {XCurve {keys [<at 0.5 1>,<at 1 0>]}, YCurve {keys [<at 0.5 1>,<at 1 0>]}, ZCurve {keys [<at 0.5 1>,<at 1 0>]}}})
    ApplyForce=false
    ApplyPosition=true
    ApplyPositionFalloff=false
   )
   &SolveForcesAndVelocity(Clamp Velocity=false, Limit Acceleration=false)
   &ScaleColor(
    Scale Alpha=&FloatFromCurve(FloatCurve=DataInterface() {propertyValues {Curve {keys [<at 0.9 1>,<at 1 0>]}}})
    Scale RGB=&MakeVectorFromLinearColorRGB(Color=&FNTimeOfDayColor(
      Night Color=<rgba 0.4796 0.1454 0.619 1>
      Morning Color=<rgba 1 0.1687 0.4166 1>
      Day Color=<rgba 1 0.1687 0.4166 1>
      Evening Color=<rgba 0.4796 0.1454 0.619 1>
     ))
    ScaleA=true
    ScaleRGB=true
   )
  ]
  rendererClasses [&NiagaraMeshRendererProperties, &NiagaraSpriteRendererProperties]
  renderers [
   {rendererClass=&NiagaraMeshRendererProperties, rendererIndex=0}
   {rendererClass=&NiagaraSpriteRendererProperties, rendererIndex=1}
  ]
 }
 Empty {
  bEnabled=false
  emitterUpdateScript [
   &EmitterState()
   &SpawnBurst_Instantaneous(Use Spawn Probability=false, UseLoopCountLimit=false)
  ]
  particleSpawnScript [&InitializeParticle(Color Mode=::NewEnumerator0, Lifetime=5, Sprite Size Mode=::NewEnumerator3, UsePositionOffset=false, Write Material Random=true)]
  particleUpdateScript [&ParticleState()]
  rendererClasses [&NiagaraSpriteRendererProperties, &NiagaraMeshRendererProperties]
  renderers [
   {rendererClass=&NiagaraSpriteRendererProperties}
   {rendererClass=&NiagaraMeshRendererProperties, rendererIndex=1}
  ]
 }
 FN_Contrails_Dots {
  emitterUpdateScript [
   &EmitterState()
   &SpawnBurst_Instantaneous(
    Spawn Count=&RandomRangeInt(Minimum=3, Maximum=5)
    Use Spawn Probability=false
    UseLoopCountLimit=false
   )
  ]
  particleSpawnScript [
   &InitializeParticle(
    Color Minimum=<rgba 1 0.1687 0.388 1>
    Color Mode=::NewEnumerator2
    Lifetime Max=0.5
    Lifetime Min=0.3
    Lifetime Mode=::NewEnumerator1
    Sprite Rotation Mode=::NewEnumerator2
    Sprite Size Mode=::NewEnumerator1
    Uniform Sprite Size Max=5
    Uniform Sprite Size Min=3
    UsePositionOffset=false
    Write Material Random=true
   ) {module=InitializeParticle001}
   &AddVelocity(Velocity=&RandomRangeVector(Minimum=<0 100 0>, Maximum=<20 150 0>))
  ]
  particleUpdateScript [
   &ParticleState()
   &ScaleSpriteSize(Uniform Curve Sprite Scale=DataInterface() {propertyValues {Curve {keys [
        {interpMode=RCIM_Cubic, tangentMode=RCTM_User, time=0, value=0}
        {interpMode=RCIM_Cubic, tangentMode=RCTM_User, time=0.25, value=1}
        {arriveTangent=-1.3333, interpMode=RCIM_Cubic, leaveTangent=-1.3333, time=0.85, value=1}
        {tangentMode=RCTM_User, time=1, value=0}
       ]}}})
   /Niagara/Modules/Update/Forces/CurlNoiseForce(
    Noise Frequency=0.1
    Noise Strength=&SelectFloatFromArray(
     Direct Array Index=$Particles.PartitionParticles.Partition
     Float Selection Array=DataInterface() {propertyValues {FloatData [500,500,200,200]}}
    )
    Pan Noise=false
   )
   &InheritSourceMovement(
    Applied Position Scale=&VectorFromCurve(VectorCurve=DataInterface() {propertyValues {XCurve {keys [<at 0.5 1>,<at 1 0>]}, YCurve {keys [<at 0.5 1>,<at 1 0>]}, ZCurve {keys [<at 0.5 1>,<at 1 0>]}}})
    ApplyForce=false
    ApplyPosition=true
    ApplyPositionFalloff=false
   )
   &SolveForcesAndVelocity(Clamp Velocity=false, Limit Acceleration=false)
   &ScaleColor(
    Scale Alpha=&FloatFromCurve(FloatCurve=DataInterface() {propertyValues {Curve {keys [<at 0.9 1>,<at 1 0>]}}})
    Scale RGB=&MakeVectorFromLinearColorRGB(Color=&FNTimeOfDayColor(
      Night Color=<rgba 0.4796 0.1454 0.619 1>
      Morning Color=<rgba 1 0.1687 0.4166 1>
      Day Color=<rgba 1 0.1687 0.4166 1>
      Evening Color=<rgba 0.4796 0.1454 0.619 1>
     ))
    ScaleA=true
    ScaleRGB=true
   )
  ]
  rendererClasses [&NiagaraSpriteRendererProperties]
  renderers [{rendererClass=&NiagaraSpriteRendererProperties}]
 }
}
summary {
 systemName=NS_Emote_SacredCuddle_BlowKiss
 emitters [
  {emitterName=BigHeart, rendererClasses [&NiagaraMeshRendererProperties, &NiagaraSpriteRendererProperties]}
  {
   bEnabled=false
   emitterName=Empty
   rendererClasses [&NiagaraSpriteRendererProperties, &NiagaraMeshRendererProperties]
  }
  {emitterName=FN_Contrails_Dots, rendererClasses [&NiagaraSpriteRendererProperties]}
 ]
 userVariables [
  User.bIsFrontEnd {type=NiagaraBool}
  User.bIsFrontEndPreview {type=NiagaraBool}
 ]
}
